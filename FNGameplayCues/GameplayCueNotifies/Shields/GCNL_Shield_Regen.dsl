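; Looping gameplay cue notify that drives the visual effect for shield regeneration.
; When the loop starts, it safely fetches the first spawned particle component.
; If the cue's raw magnitude is non-negative, it writes that value to the effect as a float parameter, named by the shield amount parameter name variable.
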(event EventOnLoopingStart(EitherVFX-preferthisversion)
  (params (MyTarget :type Actor%20Object%20Reference) (Parameters :type Gameplay%20Cue%20Parameters%20Structure%20%28by%20ref%29) (ParticleComponents :type Array%20of%20FXSystem%20Component%20Object%20References) (AudioComponents :type Array%20of%20Audio%20Component%20Object%20References))
  (bind (_normalizedmagnitude _rawmagnitude) (Ability|GameplayCue|BreakGameplayCueParameters Parameters))
  (bind _result (ArrayMacros|SafeArrayGet ParticleComponents)
    (:%20%20
      (if (>= _rawmagnitude 0)
        (Effects|Components|ParticleSystem|SetFloatParameter _result (Variables|Default|GetShieldAmountParamName) _rawmagnitude)))))

(fn ConstructionScript)
